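# Physics asset for the Festive Pug backpack cosmetic, giving its bow and ribbon bones collision bodies and angular limits.
# Sphere, capsule and box bodies cover the bow, its two side pieces, two ribbon chains and a small pack capsule.
# Both side pieces and both two-bone ribbon chains hang off dyn_bow, with tighter swing and twist limits at the ribbon roots.
# No constraint links dyn_bow to the pack body.
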PhysicsAsset /BRCosmetics/Accessories/FORT_Backpacks/Backpack_M_MED_Festive_Pug/Meshes/M_MED_Festive_Pug_Physics
bodies {
 dyn_bow [
  {shapeType=Sphere, radius=5.7538, center=<-7.5181 0 0>}
  {shapeType=Capsule, length=2.364, radius=6.3117, center=<-7.0327 -5.5311 0.2163>}
  {shapeType=Capsule, length=4.2009, radius=5.4184, center=<-6.2788 -3.8624 0>}
 ]
 dyn_bow_l [{shapeType=Capsule, length=1.2556, radius=1.9595, center=<1.6709 0 0.2759>, rotation=<rot 59.6029 179.9999 179.9999>}]
 dyn_bow_r [{shapeType=Capsule, length=1.2556, radius=2.0644, center=<-2.2455 0 0>, rotation=<rot 90 0 0>}]
 dyn_ribbon [{shapeType=Box, extentX=3.1607, extentY=1.01, extentZ=2.2657, center=<0.9902 0 0>, rotation=<rot 2.5226 -0.0001 0>}]
 dyn_ribbon1 [{shapeType=Capsule, length=2, radius=0.5, center=<1 0 0>, rotation=<rot 90 0 0>}]
 dyn_ribbon_l [{shapeType=Box, extentX=5.4372, extentY=0.8724, extentZ=3.6364, center=<2.4556 0.2049 0.4389>, rotation=<rot 3.1621 -1.8165 -13.1179>}]
 dyn_ribbon_r [{shapeType=Box, extentX=3.4802, extentY=1.5468, extentZ=3.4132, center=<2.4764 -0.4037 0.5301>, rotation=<rot 3.9517 -5.1039 -9.6183>}]
 pack [{shapeType=Capsule, length=0.4953, radius=1.2347, shapeName=pack_capsule, center=<0.0084 -16.337 4.6875>, rotation=<rot 0 -90 0>}]
}
constraints [
 {bone1Name=dyn_bow_l, bone2Name=dyn_bow, swing1LimitDegrees=45, swing2LimitDegrees=45, twistLimitDegrees=15}
 {bone1Name=dyn_bow_r, bone2Name=dyn_bow, swing1LimitDegrees=45, swing2LimitDegrees=45, twistLimitDegrees=15}
 {bone1Name=dyn_ribbon, bone2Name=dyn_bow, swing1LimitDegrees=30, swing2LimitDegrees=30, twistLimitDegrees=5}
 {bone1Name=dyn_ribbon_r, bone2Name=dyn_ribbon, swing1LimitDegrees=45, swing2LimitDegrees=45, twistLimitDegrees=15}
 {bone1Name=dyn_ribbon1, bone2Name=dyn_bow, swing1LimitDegrees=30, swing2LimitDegrees=30, twistLimitDegrees=5}
 {bone1Name=dyn_ribbon_l, bone2Name=dyn_ribbon1, swing1LimitDegrees=45, swing2LimitDegrees=45, twistLimitDegrees=15}
]
